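# Conversation database for the MoonHawk NPC, defining its greeting and dialogue flow from the Conversation.Entry.MoonHawk entry tag.
# A speech node shows a default heist-joke line or contextual variants for a first-ever conversation or players wearing specific character skins.
# The first-meeting line takes its priority from a curve row, while the skin lines clear the curve reference.
# Speech then branches through link nodes into the shared NPCLibrary quests and basic services conversations.
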ConversationDatabase /NPCLibrary/NPCs/MoonHawk/MoonHawk/NPCLibrary_Conversation_MoonHawk
entry_points [{tag=Conversation.Entry.MoonHawk, destinations [#1]}]
nodes [
 ConversationEntryPointNode_7 {
  id=#1
  class=ConversationEntryPointNode
  outputs [#2]
  properties {entryTag {tagName=Conversation.Entry.MoonHawk}}
 }
 Speech_C_4 {
  id=#2
  class=Speech_C
  outputs [#3, #4]
  properties {
   nodeName=""
   generalConfig {
    defaultMessage="I'm working on our next big heist-- I mean, military plan."
    contextualMessages [
     {
      message=You look like you're looking for danger. I can help with that.
      contextRequirements [{requirement {class=FortControllerRequirement_IsFirstEverConversationWithNPC}}]
      priority {curve {rowName=Tandem.Speech.Priority.FirstEverInteraction}}
     }
     {
      message="I'm going to uncover my past, with or without your help."
      contextRequirements [{requirement {
         class=FortControllerRequirement_HasCID
         properties {allowedCIds [/BRCosmetics/Athena/Items/Cosmetics/Characters/CID_A_286_Athena_Commando_M_Turtleneck]}
        }}]
      priority {value=0, curve {curveTable=null, rowName=None}}
     }
     {
      message=" I'm going to help you bring her home."
      contextRequirements [{requirement {
         class=FortControllerRequirement_HasCID
         properties {allowedCIds [/BRCosmetics/Athena/Items/Cosmetics/Characters/CID_540_Athena_Commando_M_MeteorManRemix]}
        }}]
      priority {value=0, curve {curveTable=null, rowName=None}}
     }
     {
      message="What are you lookin' at, mister bashful?"
      contextRequirements [{requirement {
         class=FortControllerRequirement_HasCID
         properties {allowedCIds [/BRCosmetics/Athena/Items/Cosmetics/Characters/CID_140_Athena_Commando_M_Visitor]}
        }}]
      priority {value=0, curve {curveTable=null, rowName=None}}
     }
     {
      message=Your last message has been delivered. Don't worry.
      contextRequirements [{requirement {
         class=FortControllerRequirement_HasCID
         properties {allowedCIds [/BRCosmetics/Athena/Items/Cosmetics/Characters/CID_547_Athena_Commando_F_Meteorwoman]}
        }}]
      priority {value=0, curve {curveTable=null, rowName=None}}
     }
     {
      message="Me? Insubordinate? Never!"
      contextRequirements [{requirement {
         class=FortControllerRequirement_HasCID
         properties {allowedCIds [/BRCosmetics/Athena/Items/Cosmetics/Characters/CID_A_373_Athena_Commando_M_OriginPrisoner]}
        }}]
      priority {value=0, curve {curveTable=null, rowName=None}}
     }
    ]
   }
  }
 }
 ConversationLinkNode_0 {id=#4, class=ConversationLinkNode, properties {remoteEntryTag {tagName=Conversation.Entry.BasicServices}}}
 ConversationLinkNode_1 {id=#3, class=ConversationLinkNode, properties {remoteEntryTag {tagName=Conversation.Entry.NPCLibrary.Quests}}}
]
